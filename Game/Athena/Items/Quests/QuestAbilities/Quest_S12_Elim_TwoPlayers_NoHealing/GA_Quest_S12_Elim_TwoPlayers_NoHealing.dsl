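; Quest ability for the Season 12 "eliminate two players without healing" challenge, triggered by events whose target carries the player tag.
; On activation it runs parent quest logic, snapshots the owning pawn's health and shield, binds to health changes, and applies a NoHealing effect if absent.
; The health callback lowers the snapshots, removing the effect and unbinding when health rises without a shield increase.
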(event Ability|EventActivateAbilityFromEvent
  (params (EventData :type Gameplay%20Event%20Data%20Structure%20%28by%20ref%29))
  (bind _owningplayerpawn (Variables|GAVars|Data|GetOwningPlayerPawn))
  ;[1 1 1] On Initial Kill Apply NoHealing Tag, and Update Quest
  (|CallParentFunction EventData)
  (Variables|Default|SetPlayerShield (FortPawn|GetShield _owningplayerpawn))
  (Variables|Default|SetPlayerHealth (FortPawn|GetHealth _owningplayerpawn))
  (Health|AssignOnPawnHealthChanged _owningplayerpawn (delegate Custom|OnPawnHealthChanged))
  (if (not (Ability|DoesTargetHaveTags _owningplayerpawn (Variables|Default|GetT_NoHealing)))
    (GameplayEffects|ApplyGameplayEffectToSelf :self (Ability|GetAbilitySystemComponent _owningplayerpawn) :GameplayEffectClass "/Game/Athena/Items/Quests/QuestAbilities/Quest_S12_Elim_TwoPlayers_NoHealing/GE_Quest_S12_Elim_TwoPlayers_NoHealing.GE_Quest_S12_Elim_TwoPlayers_NoHealing_C" :Level 0.0))
  (Ability|EndAbility))

(event Custom|OnPawnHealthChanged
  (bind _returnvalue (FortPawn|GetHealth (Variables|GAVars|Data|GetOwningPlayerPawn)))
  (bind _owningplayerpawn (Variables|GAVars|Data|GetOwningPlayerPawn))
  (bind _returnvalue_2 (FortPawn|GetShield _owningplayerpawn))
  ;[1 1 1] On Health Change, Make Sure Player Didn't Heal
  (if (not (> _returnvalue_2 (Variables|Default|GetPlayerShield)))
    (Variables|Default|SetPlayerShield _returnvalue_2)
    (if (> _returnvalue (Variables|Default|GetPlayerHealth))
      (GameplayEffects|RemoveActiveEffectswithGrantedTags (Ability|GetAbilitySystemComponent _owningplayerpawn) (Variables|Default|GetT_NoHealing))
      (Health|UnbindEventfromOnPawnHealthChanged (Variables|GAVars|Data|GetOwningPlayerPawn) (delegate Custom|OnPawnHealthChanged))
      (else
        (Variables|Default|SetPlayerHealth _returnvalue)))
    (Ability|EndAbility)))

(fn ShouldAbilityRespondToEvent
  (params (ActorInfo :type Gameplay%20Ability%20Actor%20Info%20Structure) (Payload_EventTag :type Gameplay%20Tag%20Structure) (Payload_Instigator :type Actor%20Object%20Reference) (Payload_Target :type Actor%20Object%20Reference) (Payload_OptionalObject :type Object%20Reference) (Payload_OptionalObject2 :type Object%20Reference) (Payload_ContextHandle :type Gameplay%20Effect%20Context%20Handle%20Structure) (Payload_InstigatorTags :type Gameplay%20Tag%20Container%20Structure) (Payload_TargetTags :type Gameplay%20Tag%20Container%20Structure) (Payload_EventMagnitude :type Float%20%28single-precision%29) (Payload_TargetData :type Gameplay%20Ability%20Target%20Data%20Handle%20Structure) (Payload_InstancedEventData :type Instanced%20Struct%20Structure))
  (return (GameplayTags|HasTag Payload_TargetTags (Variables|Default|GetT_PlayerTag))))
